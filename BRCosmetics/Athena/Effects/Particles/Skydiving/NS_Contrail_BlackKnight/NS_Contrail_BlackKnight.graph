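# Niagara skydiving contrail for Black Knight, combining four body ribbons with left and right flag meshes, windburst meshes and cloth-scrap sprites.
# Flags, windbursts and scraps sample bones on User.Flag_SkeletalMesh; persistent single-particle meshes orient to the owner, with dive-driven material noise.
# Fast descent (owner Z velocity below -2600) gates ribbon and scrap spawning and intensifies mesh motion.
# The original LT_Flag and RT_Flag emitters are disabled; their 001 variants are active.
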NiagaraSystem /BRCosmetics/Athena/Effects/Particles/Skydiving/NS_Contrail_BlackKnight/NS_Contrail_BlackKnight
dependencies {
 usedDataInterfaces [&NiagaraDataInterfaceVector2DCurve, &NiagaraDataInterfaceCurve, &NiagaraDataInterfaceColorCurve]
 usedDynamicInputs []
 usedModules [
  &SystemState
  &self:SystemSpawnScript.SystemScriptSource.SystemScriptGraph.NiagaraNodeAssignment_1.SetVariables_#1
  &self:SystemSpawnScript.SystemScriptSource.SystemScriptGraph.NiagaraNodeAssignment_0.SetVariables_#2
 ]
}
emitters {
 Basic_Contrail {
  emitterUpdateScript [
   &EmitterScalability(
    Module.bApplySpawnCountScale=false
    Module.bApplySpawnCountScaleByDistanceFraction=false
    Module.bUseMaxDistance=false
    Module.bUseMinDistance=false
   ) {enabled=false}
   &EmitterLifeCycle(CompleteOnInactive=true) {enabled=false}
   &EmitterState()
   /Game/Effects/Niagara/Modules/Spawn/FN_SpawnContrails(All limbs=false, Global Spawnrate=60) {enabled=false}
   /Game/Effects/Niagara/Modules/AnimTrail/FN_SpawnAnimTrail/FN_SpawnAnimTrail() {enabled=false}
   &SpawnRate(SpawnRate=30, Use Spawn Probability=false) {enabled=false}
   &FN_SpawnRibbon_v2(
    Number of Ribbons=4
    Spawn Mode=::NewEnumerator0
    SpawnRate=HlslExpression() {hlslExpression="(Engine.Owner.Velocity.z < -2600.0f) ? 40 : 0"}
   )
  ]
  particleSpawnScript [
   &FN_InitializeRibbon_v3(Enable Skeletal Position=true, Lifetime=0.7, Ribbon Width=60) {enabled=false}
   &FN_InitializeRibbon_v3(
    Color:LinearColor=<rgba 1 1 1 0.1>
    Enable Skeletal Position=true
    Lifetime=1
    Sampled Position Coordinate Space:ENiagaraCoordinateSpace=::Local
    Sampled Rotation Coordinate Space:ENiagaraCoordinateSpace=::Local
    Sampled Velocity Coordinate Space:ENiagaraCoordinateSpace=::Local
   ) {module=FN_InitializeRibbon_v3001}
   &AddVelocity(Velocity=<0 0 410>) {enabled=false}
   &AddVelocity(Velocity=&MakeCustomVectorFromBool(
     Bool=$Particles.Locked
     True Vector=<0 0 0>
     False Vector=&FN_FrontEndVector3Modifier(Gameplay=<0 0 100>, Front End=<0 0 350>)
    )) {module=AddVelocity001}
   &DynamicMaterialParameters(
    Index 0 Param 2=0.1
    Index 0 Param 3=1
    Index 0 Param 4=1
    Module.Param0WriteEnabled=true
    Module.Param1WriteEnabled=true
    Module.Param2WriteEnabled=true
    Module.Param3WriteEnabled=true
   )
  ]
  particleUpdateScript [
   &Drag(Drag=0.9, Use Linear Drag=true, Use Rotational Drag=false)
   &InheritVelocity(
    Module.Limit Speed=true
    Velocity Limit=10000
    Velocity Scale=&MakeCustomVectorFromBool(Bool=$Particles.Initial.Locked, True Vector=<0 0 0>, False Vector=<-0.03 -0.03 -0.01>)
    Velocity Source=&TransformVector(Vector=$Engine.Owner.Velocity, Source Space=::World, Destination Space=::Local)
   )
   /Game/Effects/Niagara/Modules/Skydiving/FN_DivingAttributes/FN_DivingAttributes() {enabled=false}
   &SolveForcesAndVelocity(Clamp Velocity=false, Limit Acceleration=false)
   &FN_UpdateRibbon(
    Enable Skeletal Position=true
    Initial Color:LinearColor=$Particles.Initial.Color
    Initial RibbonWidth:NiagaraFloat=$Particles.Initial.RibbonWidth
    Sampled Position Coordinate Space:ENiagaraCoordinateSpace=::Local
    Sampled Rotation Coordinate Space:ENiagaraCoordinateSpace=::Local
    Sampled Velocity Coordinate Space:ENiagaraCoordinateSpace=::Local
   )
   &Material(Base Texture Time:NiagaraFloat=&Multiply_Float(A=$Emitter.Age, B=-1)) {enabled=false}
   &ScaleRibbonWidth(Ribbon Width Scale=&FloatFromCurve(FloatCurve=DataInterface() {propertyValues {Curve {keys [
         {arriveTangent=1.9737, interpMode=RCIM_Cubic, leaveTangent=1.9737, tangentMode=RCTM_User, time=0.0058, value=0.9363}
         {arriveTangent=-0.0131, interpMode=RCIM_Cubic, leaveTangent=-0.0131, time=0.4538, value=1.5093}
         <at 0.6635 0.9277>
        ]}}}))
   &Color(
    Color=&ColorFromCurve(ColorCurve=DataInterface() {propertyValues {
       AlphaCurve {keys [<at 0.1 0 RCIM_Cubic>,<at 0.3 1 RCIM_Cubic>,<at 1 0>]}
       BlueCurve {keys [<at 0 1 RCIM_Cubic>,<at 1 1 RCIM_Cubic>]}
       GreenCurve {keys [<at 0 1 RCIM_Cubic>,<at 1 1 RCIM_Cubic>]}
       RedCurve {keys [<at 0 1 RCIM_Cubic>,<at 1 1 RCIM_Cubic>]}
      }})
    Scale Alpha=&FloatFromCurve(
     FloatCurve=DataInterface() {propertyValues {Curve {keys [
         {arriveTangent=2.7939, interpMode=RCIM_Cubic, leaveTangent=2.7939, tangentMode=RCTM_User, time=-0.0046, value=1.0113}
         {arriveTangent=-1.0036, interpMode=RCIM_Cubic, leaveTangent=-4.3444, tangentMode=RCTM_Break, time=0.1387, value=1.232}
         <at 0.6 0.002>
        ]}}}
     Scale Curve=0.8
    )
   )
  ]
  rendererClasses [&NiagaraRibbonRendererProperties]
  renderers [{rendererClass=&NiagaraRibbonRendererProperties}]
 }
 LT_ClothPieces {
  emitterUpdateScript [
   &EmitterState(Life Cycle Mode=::NewEnumerator1, Loop Duration=1, UseLoopDelay=false)
   &SpawnRate(
    Spawn Probability=0.1
    SpawnRate=HlslExpression() {hlslExpression="(Engine.Owner.Velocity.z < -2600.0f) ? 4.0f : 0.0f"}
    Use Spawn Probability=true
   )
  ]
  particleSpawnScript [
   /Niagara/Modules/Spawn/Initialization/InitializeParticle(
    Color=<rgba 0.013 0.0152 0.0242 1>
    Lifetime=&UniformRangedFloat(Minimum=1.8, Maximum=2)
    Mass=&UniformRangedFloat(Minimum=0.75, Maximum=1.75)
    Module.Write Color=true
    Module.Write Lifetime=true
    Module.Write Mass=true
    Module.Write Material Random=true
    Module.Write Position=false
    Module.Write Scale=false
    Module.Write SpriteRotation=false
    Module.Write SpriteSize=true
    Position=&SimulationPosition()
    Sprite Size=&UniformRangedVector2D(Minimum=<7 7>, Maximum=<7 15>)
   )
   /Game/Effects/Niagara/Modules/Groups/FN_SetSpawnGroup_v1/FN_SetSpawnGroup(Override SpawnGroup=true)
   /SetVariables(Particles.SubImageIndex=&UniformRangedFloat(Minimum=0, Maximum=3)) {isSetParameters=true, module=SetVariables_#3}
   &AddVelocity(Velocity=&UniformRangedVector(Minimum=<-20 -20 150>, Maximum=<20 20 200>))
   &SampleSkeletalMeshSkeleton(
    Mesh:NiagaraDataInterfaceSkeletalMesh=$User.Flag_SkeletalMesh
    Sampled Position Coordinate Space:ENiagaraCoordinateSpace=::Local
    Sampled Rotation Coordinate Space:ENiagaraCoordinateSpace=::Local
    Sampled Velocity Coordinate Space:ENiagaraCoordinateSpace=::Local
    Specific Bone Index:NiagaraInt32=$Particles.SpawnGroup
   )
   /Niagara/Modules/Spawn/Location/SkeletalMeshSkeletonLocation(Module.Inherit Velocity=false, Module.Use Sampled Bone as Ribbon ID=true)
   &BoxLocation(Box Size=<7 1 0>, MaskSpawnGroup=false, Module.Override Seed=false, Rotate Box=false, Surface Only=false)
  ]
  particleUpdateScript [
   &ParticleState()
   &AccelerationForce(Acceleration=<0 0 2100>)
   /Niagara/Modules/Update/Forces/CurlNoiseForce(Noise Strength=250, Pan Noise=false) {module=CurlNoiseForce001}
   &SolveForcesAndVelocity(Clamp Velocity=false, Limit Acceleration=false)
   &ScaleSpriteSize(Scale Factor=&Vector2DFromCurve(Vector2Curve=DataInterface() {propertyValues {XCurve {keys [<at 0 1>,<at 1 0>]}, YCurve {keys [<at 0 1>,<at 1 0>]}}}, ScaleCurve=<1 1>))
  ]
  rendererClasses [&NiagaraSpriteRendererProperties]
  renderers [{rendererClass=&NiagaraSpriteRendererProperties}]
 }
 LT_Flag {
  bEnabled=false
  emitterSpawnScript [&FN_PreserveSkydiveUserVariables() {enabled=false, module=PreserveUserVariables}]
  emitterUpdateScript [
   &EmitterState(
    Inactive Response=::NewEnumerator1
    Life Cycle Mode=::NewEnumerator1
    Loop Behavior=::NewEnumerator1
    Loop Duration Mode=::NewEnumerator1
    UseLoopDelay=false
   )
   &SpawnBurst_Instantaneous(Spawn Count=1, Use Spawn Probability=false)
  ]
  particleSpawnScript [
   &FN_SetSpawnGroup(Override SpawnGroup=true, SpawnGroup=1)
   &InitializeParticle(
    Lifetime=100
    Mesh Scale=&MakeVector(X=0.8, Y=1, Z=1.65)
    Module.Write Color=true
    Module.Write Lifetime=false
    Module.Write Mass=true
    Module.Write Material Random=true
    Module.Write Position=false
    Module.Write Scale=true
    Module.Write SpriteRotation=true
    Module.Write SpriteSize=true
    Position=&SimulationPosition()
    Sprite Rotation=&UniformRangedFloat(Minimum=0, Maximum=360)
   )
   /Niagara/Modules/Spawn/Orientation/V2/InitialMeshOrientation(
    Orientation Coordinate Space:ENiagaraCoordinateSpace=::Local
    Orientation Vector=<0 90 0>
    Use Orientation Vector=true
    Use Rotation Vector=false
   )
   &DynamicMaterialParameters(
    Index 1 Param 3=0.25
    Module.Param0WriteEnabled1=true
    Module.Param1WriteEnabled1=true
    Module.Param2WriteEnabled1=true
    Module.Param3WriteEnabled1=true
    Write Parameter Index 0=false
    Write Parameter Index 1=true
   )
  ]
  particleUpdateScript [
   /Game/Effects/Niagara/Modules/Skydiving/FN_DivingAttributes/FN_DivingAttributes_v2(SkeletalMesh:NiagaraDataInterfaceSkeletalMesh=$User.Flag_SkeletalMesh)
   &ParticleState(Kill Particles When Lifetime Has Elapsed=false)
   /Niagara/Modules/Update/Orientation/V2/UpdateMeshOrientation(Rotation Coordinate Space:ENiagaraMeshTransforms=::NewEnumerator2, Rotation Rate=0, Rotation Vector:Vector3f=<0 90 0>)
   &SampleSkeletalMeshSkeleton(
    Mesh:NiagaraDataInterfaceSkeletalMesh=$User.Flag_SkeletalMesh
    Sampled Position Coordinate Space:ENiagaraCoordinateSpace=::Local
    Sampled Rotation Coordinate Space:ENiagaraCoordinateSpace=::Local
    Sampled Velocity Coordinate Space:ENiagaraCoordinateSpace=::Local
    Spawn Mode:ENiagaraSkeletalSpawning=::NewEnumerator1
    Specific Bone Index:NiagaraInt32=$Particles.SpawnGroup
   )
   &SkeletalMeshSkeletonLocation(Module.Inherit Velocity=false, Module.Use Sampled Bone as Ribbon ID=true)
   /Game/Effects/Niagara/Modules/Position/FN_OffsetPositionByOrientationAxis(Offset Attribute:Vector3f=$Particles.SampleSkeletalMeshSkeleton.SampledPosition, Offset Axis:Vector3f=<7 7 -10>, Z Symmetry=true) {module=OffsetPositionByMeshOrientation}
   /BRCosmetics/Athena/Effects/Particles/Skydiving/NS_Contrail_AmmoBelt/Module/FN_OrientMeshToEngineOwner(
    Direction:Vector3f=&FN_FrontEndVector3Modifier(Gameplay=<0 0 -1>, Front End=<0 0 -1>)
    Mesh Forward Vector:Vector3f=<-0.2 1 0>
    Set Mesh Orientation=true
    Vector:Vector3f=&TransformVector(Vector=$Engine.Owner.SystemYAxis, Source Space=::World, Destination Space=::Simulation, Apply Scale=true)
   )
   /Game/Effects/Niagara/Modules/Materials/FN_MeshMaterial(
    Amplitude=80
    Falloff=HlslExpression() {hlslExpression="(Engine.Owner.Velocity.z < -2600.0f) ? 600.0f : 350.0f"}
    Frequency=300
    Jitter Multiplier=1.8545
    Offset:NiagaraFloat=&Add_Float(A=$Particles.AccumulatedNormalizedDive, B=0.3)
    Scale Noise X Axis=1.3
    Scale Noise Y Axis=2
    Scale Noise Z Axis=0
    Speed=HlslExpression() {hlslExpression="(Engine.Owner.Velocity.z < -2600.0f) ? 1.0f : 0.5f"}
   )
   &DynamicMaterialParameters(
    Index 2 Param 1=0.6
    Index 2 Param 2=1.1
    Index 2 Param 3=0.1
    Module.Param0WriteEnabled2=true
    Module.Param1WriteEnabled2=true
    Module.Param2WriteEnabled2=true
    Module.Param3WriteEnabled2=false
    Write Parameter Index 0=false
    Write Parameter Index 2=true
   ) {module=DynamicMaterialParameters001}
   &SolveForcesAndVelocity(Clamp Velocity=false, Limit Acceleration=false)
   /SetVariables(Particles.MeshOrientation=&MultiplyQuaternion(
     Quaternion A=$Particles.MeshOrientation
     Quaternion B=&MakeQuatFromAxisAngle(Axis=<0 0 0>, AngleInDegrees=&Multiply_Float(A=$Particles.Initial.SpriteRotation, B=1))
    )) {isSetParameters=true, module=SetVariables_#4}
  ]
  rendererClasses [&NiagaraMeshRendererProperties]
  renderers [{rendererClass=&NiagaraMeshRendererProperties, rendererIndex=0}]
 }
 LT_Flag001 {
  emitterSpawnScript [&FN_PreserveSkydiveUserVariables() {enabled=false, module=PreserveUserVariables}]
  emitterUpdateScript [
   &EmitterState(
    Inactive Response=::NewEnumerator1
    Life Cycle Mode=::NewEnumerator1
    Loop Behavior=::NewEnumerator1
    Loop Duration Mode=::NewEnumerator1
    UseLoopDelay=false
   )
   &SpawnBurst_Instantaneous(Spawn Count=1, Use Spawn Probability=false)
  ]
  particleSpawnScript [
   &FN_SetSpawnGroup(Override SpawnGroup=true, SpawnGroup=1)
   &InitializeParticle(
    Lifetime=100
    Mesh Scale=&MakeVector(X=0.8, Y=1, Z=2)
    Module.Write Color=true
    Module.Write Lifetime=false
    Module.Write Mass=true
    Module.Write Material Random=true
    Module.Write Position=false
    Module.Write Scale=true
    Module.Write SpriteRotation=true
    Module.Write SpriteSize=true
    Position=&SimulationPosition()
    Sprite Rotation=&UniformRangedFloat(Minimum=0, Maximum=360)
   )
   &InitialMeshOrientation(
    Orientation Coordinate Space:ENiagaraCoordinateSpace=::Local
    Orientation Vector=<0 90 0>
    Use Orientation Vector=true
    Use Rotation Vector=false
   )
   &DynamicMaterialParameters(
    Index 1 Param 3=0.25
    Module.Param0WriteEnabled1=true
    Module.Param1WriteEnabled1=true
    Module.Param2WriteEnabled1=true
    Module.Param3WriteEnabled1=true
    Write Parameter Index 0=false
    Write Parameter Index 1=true
   )
  ]
  particleUpdateScript [
   &FN_DivingAttributes_v2(SkeletalMesh:NiagaraDataInterfaceSkeletalMesh=$User.Flag_SkeletalMesh)
   &ParticleState(Kill Particles When Lifetime Has Elapsed=false)
   &UpdateMeshOrientation(Rotation Coordinate Space:ENiagaraMeshTransforms=::NewEnumerator2, Rotation Rate=0, Rotation Vector:Vector3f=<0 90 0>)
   &SampleSkeletalMeshSkeleton(
    Mesh:NiagaraDataInterfaceSkeletalMesh=$User.Flag_SkeletalMesh
    Sampled Position Coordinate Space:ENiagaraCoordinateSpace=::Local
    Sampled Rotation Coordinate Space:ENiagaraCoordinateSpace=::Local
    Sampled Velocity Coordinate Space:ENiagaraCoordinateSpace=::Local
    Spawn Mode:ENiagaraSkeletalSpawning=::NewEnumerator1
    Specific Bone Index:NiagaraInt32=$Particles.SpawnGroup
   )
   &SkeletalMeshSkeletonLocation(Module.Inherit Velocity=false, Module.Use Sampled Bone as Ribbon ID=true)
   &FN_OffsetPositionByOrientationAxis(Offset Attribute:Vector3f=$Particles.SampleSkeletalMeshSkeleton.SampledPosition, Offset Axis:Vector3f=<7 5 -10>, Z Symmetry=true) {module=OffsetPositionByMeshOrientation}
   &FN_OrientMeshToEngineOwner(
    Direction:Vector3f=&FN_FrontEndVector3Modifier(Gameplay=<0 0 -1>, Front End=<0 0 -1>)
    Mesh Forward Vector:Vector3f=<-0.2 1 0>
    Set Mesh Orientation=true
    Vector:Vector3f=&TransformVector(Vector=$Engine.Owner.SystemYAxis, Source Space=::World, Destination Space=::Simulation, Apply Scale=true)
   )
   &FN_MeshMaterial(
    Amplitude=80
    Falloff=HlslExpression() {hlslExpression="(Engine.Owner.Velocity.z < -2600.0f) ? 600.0f : 350.0f"}
    Frequency=300
    Jitter Multiplier=1.8545
    Offset:NiagaraFloat=&Add_Float(A=$Particles.AccumulatedNormalizedDive, B=0.3)
    Scale Noise X Axis=1.3
    Scale Noise Y Axis=2
    Scale Noise Z Axis=0
    Speed=HlslExpression() {hlslExpression="(Engine.Owner.Velocity.z < -2600.0f) ? 1.0f : 0.5f"}
   )
   &DynamicMaterialParameters(
    Index 2 Param 1=0.6
    Index 2 Param 2=1.1
    Index 2 Param 3=0.1
    Module.Param0WriteEnabled2=true
    Module.Param1WriteEnabled2=true
    Module.Param2WriteEnabled2=true
    Module.Param3WriteEnabled2=false
    Write Parameter Index 0=false
    Write Parameter Index 2=true
   ) {module=DynamicMaterialParameters001}
   &SolveForcesAndVelocity(Clamp Velocity=false, Limit Acceleration=false)
   /SetVariables(Particles.MeshOrientation=&MultiplyQuaternion(
     Quaternion A=$Particles.MeshOrientation
     Quaternion B=&MakeQuatFromAxisAngle(Axis=<0 0 0>, AngleInDegrees=&Multiply_Float(A=$Particles.Initial.SpriteRotation, B=1))
    )) {isSetParameters=true, module=SetVariables_#4}
  ]
  rendererClasses [&NiagaraMeshRendererProperties]
  renderers [{rendererClass=&NiagaraMeshRendererProperties, rendererIndex=0}]
 }
 LT_Windburst {
  emitterSpawnScript [&FN_PreserveSkydiveUserVariables() {enabled=false, module=PreserveUserVariables}]
  emitterUpdateScript [
   &EmitterState(
    Inactive Response=::NewEnumerator1
    Life Cycle Mode=::NewEnumerator1
    Loop Behavior=::NewEnumerator1
    Loop Duration Mode=::NewEnumerator1
    UseLoopDelay=false
   )
   &SpawnBurst_Instantaneous(Spawn Count=1, Use Spawn Probability=false)
  ]
  particleSpawnScript [
   &FN_SetSpawnGroup(Override SpawnGroup=true)
   &InitializeParticle(
    Color=<rgba 1 1 1 10>
    Lifetime=100
    Mesh Scale=&MakeVector(X=0.8, Y=1, Z=3.5)
    Module.Write Color=true
    Module.Write Lifetime=false
    Module.Write Mass=true
    Module.Write Material Random=true
    Module.Write Position=false
    Module.Write Scale=true
    Module.Write SpriteRotation=true
    Module.Write SpriteSize=true
    Position=&SimulationPosition()
    Sprite Rotation=&UniformRangedFloat(Minimum=0, Maximum=360)
   )
   &InitialMeshOrientation(
    Orientation Coordinate Space:ENiagaraCoordinateSpace=::Local
    Orientation Vector=<0 90 0>
    Use Orientation Vector=true
    Use Rotation Vector=false
   )
  ]
  particleUpdateScript [
   &FN_DivingAttributes_v2(SkeletalMesh:NiagaraDataInterfaceSkeletalMesh=$User.Flag_SkeletalMesh)
   &ParticleState(Kill Particles When Lifetime Has Elapsed=false)
   &UpdateMeshOrientation(Rotation Coordinate Space:ENiagaraMeshTransforms=::NewEnumerator2, Rotation Rate=0, Rotation Vector:Vector3f=<0 90 0>)
   &SampleSkeletalMeshSkeleton(
    Mesh:NiagaraDataInterfaceSkeletalMesh=$User.Flag_SkeletalMesh
    Sampled Position Coordinate Space:ENiagaraCoordinateSpace=::Local
    Sampled Rotation Coordinate Space:ENiagaraCoordinateSpace=::Local
    Sampled Velocity Coordinate Space:ENiagaraCoordinateSpace=::Local
    Specific Bone Index:NiagaraInt32=$Particles.SpawnGroup
   )
   &SkeletalMeshSkeletonLocation(Module.Inherit Velocity=false, Module.Use Sampled Bone as Ribbon ID=true)
   &FN_OrientMeshToEngineOwner(
    Direction:Vector3f=&FN_FrontEndVector3Modifier(Gameplay=<0 0 -1>, Front End=<0 0 -1>)
    Mesh Forward Vector:Vector3f=<-0.2 1 0>
    Set Mesh Orientation=true
    Vector:Vector3f=&TransformVector(Vector=$Engine.Owner.SystemYAxis, Source Space=::World, Destination Space=::Simulation, Apply Scale=true)
   )
   &FN_MeshMaterial(
    Amplitude=40
    Falloff=HlslExpression() {hlslExpression="(Engine.Owner.Velocity.z < -2600.0f) ? 200.0f : 100.0f"}
    Frequency=300
    Jitter Multiplier=1.8545
    Offset:NiagaraFloat=$Particles.AccumulatedNormalizedDive
    Scale Noise X Axis=1.3
    Scale Noise Y Axis=2
    Scale Noise Z Axis=0
    Speed=HlslExpression() {hlslExpression="(Engine.Owner.Velocity.z < -2600.0f) ? 1.0f : 0.4f"}
   )
   &SolveForcesAndVelocity(Clamp Velocity=false, Limit Acceleration=false) {enabled=false}
   /SetVariables(Particles.MeshOrientation=&MultiplyQuaternion(
     Quaternion A=$Particles.MeshOrientation
     Quaternion B=&MakeQuatFromAxisAngle(Axis=<0 0 0>, AngleInDegrees=&Multiply_Float(A=$Particles.Initial.SpriteRotation, B=1))
    )) {isSetParameters=true, module=SetVariables_#4}
  ]
  rendererClasses [&NiagaraMeshRendererProperties]
  renderers [{rendererClass=&NiagaraMeshRendererProperties, rendererIndex=0}]
 }
 RT_ClothPieces {
  emitterUpdateScript [
   &EmitterState(Life Cycle Mode=::NewEnumerator1, Loop Duration=1, UseLoopDelay=false)
   &SpawnRate(
    Spawn Probability=0.1
    SpawnRate=HlslExpression() {hlslExpression="(Engine.Owner.Velocity.z < -2600.0f) ? 4.0f : 0.0f"}
    Use Spawn Probability=true
   )
  ]
  particleSpawnScript [
   &InitializeParticle(
    Color=<rgba 0.013 0.0152 0.0242 1>
    Lifetime=&UniformRangedFloat(Minimum=1.8, Maximum=2)
    Mass=&UniformRangedFloat(Minimum=0.75, Maximum=1.75)
    Module.Write Color=true
    Module.Write Lifetime=true
    Module.Write Mass=true
    Module.Write Material Random=true
    Module.Write Position=false
    Module.Write Scale=false
    Module.Write SpriteRotation=false
    Module.Write SpriteSize=true
    Position=&SimulationPosition()
    Sprite Size=&UniformRangedVector2D(Minimum=<3 5>, Maximum=<15 15>)
   )
   &FN_SetSpawnGroup(Override SpawnGroup=true, SpawnGroup=1)
   /SetVariables(Particles.SubImageIndex=&UniformRangedFloat(Minimum=0, Maximum=3)) {isSetParameters=true, module=SetVariables_#3}
   &AddVelocity(Velocity=&UniformRangedVector(Minimum=<-20 -20 150>, Maximum=<20 20 200>))
   &SampleSkeletalMeshSkeleton(
    Mesh:NiagaraDataInterfaceSkeletalMesh=$User.Flag_SkeletalMesh
    Sampled Position Coordinate Space:ENiagaraCoordinateSpace=::Local
    Sampled Rotation Coordinate Space:ENiagaraCoordinateSpace=::Local
    Sampled Velocity Coordinate Space:ENiagaraCoordinateSpace=::Local
    Specific Bone Index:NiagaraInt32=$Particles.SpawnGroup
   )
   &SkeletalMeshSkeletonLocation(Module.Inherit Velocity=false, Module.Use Sampled Bone as Ribbon ID=true)
   &BoxLocation(Box Size=<7 1 0>, MaskSpawnGroup=false, Module.Override Seed=false, Rotate Box=false, Surface Only=false)
  ]
  particleUpdateScript [
   &ParticleState()
   &AccelerationForce(Acceleration=<0 0 2100>)
   &CurlNoiseForce(Noise Strength=250, Pan Noise=false) {module=CurlNoiseForce001}
   &SolveForcesAndVelocity(Clamp Velocity=false, Limit Acceleration=false)
   &ScaleSpriteSize(Scale Factor=&Vector2DFromCurve(Vector2Curve=DataInterface() {propertyValues {XCurve {keys [<at 0 1>,<at 1 0>]}, YCurve {keys [<at 0 1>,<at 1 0>]}}}, ScaleCurve=<1 1>))
  ]
  rendererClasses [&NiagaraSpriteRendererProperties]
  renderers [{rendererClass=&NiagaraSpriteRendererProperties}]
 }
 RT_Flag {
  bEnabled=false
  emitterSpawnScript [&FN_PreserveSkydiveUserVariables() {enabled=false, module=PreserveUserVariables}]
  emitterUpdateScript [
   &EmitterState(
    Inactive Response=::NewEnumerator1
    Life Cycle Mode=::NewEnumerator1
    Loop Behavior=::NewEnumerator1
    Loop Duration Mode=::NewEnumerator1
    UseLoopDelay=false
   )
   &SpawnBurst_Instantaneous(Spawn Count=1, Use Spawn Probability=false)
  ]
  particleSpawnScript [
   &FN_SetSpawnGroup(Override SpawnGroup=true, SpawnGroup=1)
   &InitializeParticle(
    Lifetime=100
    Mesh Scale=&MakeVector(X=0.8, Y=1, Z=2)
    Module.Write Color=true
    Module.Write Lifetime=false
    Module.Write Mass=true
    Module.Write Material Random=true
    Module.Write Position=false
    Module.Write Scale=true
    Module.Write SpriteRotation=true
    Module.Write SpriteSize=true
    Position=&SimulationPosition()
    Sprite Rotation=&UniformRangedFloat(Minimum=0, Maximum=360)
   )
   &InitialMeshOrientation(
    Orientation Coordinate Space:ENiagaraCoordinateSpace=::Local
    Orientation Vector=<0 90 0>
    Use Orientation Vector=true
    Use Rotation Vector=false
   )
   &DynamicMaterialParameters(
    Index 1 Param 3=0.75
    Module.Param0WriteEnabled1=true
    Module.Param1WriteEnabled1=true
    Module.Param2WriteEnabled1=true
    Module.Param3WriteEnabled1=true
    Write Parameter Index 0=false
    Write Parameter Index 1=true
   )
  ]
  particleUpdateScript [
   &FN_DivingAttributes_v2(SkeletalMesh:NiagaraDataInterfaceSkeletalMesh=$User.Flag_SkeletalMesh)
   &ParticleState(Kill Particles When Lifetime Has Elapsed=false)
   &UpdateMeshOrientation(Rotation Coordinate Space:ENiagaraMeshTransforms=::NewEnumerator2, Rotation Rate=0, Rotation Vector:Vector3f=<0 90 0>)
   &SampleSkeletalMeshSkeleton(
    Mesh:NiagaraDataInterfaceSkeletalMesh=$User.Flag_SkeletalMesh
    Sampled Position Coordinate Space:ENiagaraCoordinateSpace=::Local
    Sampled Rotation Coordinate Space:ENiagaraCoordinateSpace=::Local
    Sampled Velocity Coordinate Space:ENiagaraCoordinateSpace=::Local
    Spawn Mode:ENiagaraSkeletalSpawning=::NewEnumerator1
    Specific Bone Index:NiagaraInt32=$Particles.SpawnGroup
   )
   &SkeletalMeshSkeletonLocation(Module.Inherit Velocity=false, Module.Use Sampled Bone as Ribbon ID=true)
   &FN_OffsetPositionByOrientationAxis(Offset Attribute:Vector3f=$Particles.SampleSkeletalMeshSkeleton.SampledPosition, Offset Axis:Vector3f=<7 7 10>, Z Symmetry=true) {module=OffsetPositionByMeshOrientation}
   &FN_OrientMeshToEngineOwner(
    Direction:Vector3f=&FN_FrontEndVector3Modifier(Gameplay=<0 0 -1>, Front End=<0 0 -1>)
    Mesh Forward Vector:Vector3f=<0.2 1 0>
    Set Mesh Orientation=true
    Vector:Vector3f=&TransformVector(Vector=$Engine.Owner.SystemYAxis, Source Space=::World, Destination Space=::Simulation, Apply Scale=true)
   )
   &FN_MeshMaterial(
    Amplitude=80
    Falloff=HlslExpression() {hlslExpression="(Engine.Owner.Velocity.z < -2600.0f) ? 600.0f : 350.0f"}
    Frequency=300
    Jitter Multiplier=1.8545
    Offset:NiagaraFloat=$Particles.AccumulatedNormalizedDive
    Scale Noise X Axis=1.3
    Scale Noise Y Axis=2
    Scale Noise Z Axis=0
    Speed=HlslExpression() {hlslExpression="(Engine.Owner.Velocity.z < -2600.0f) ? 1.0f : 0.5f"}
   )
   &DynamicMaterialParameters(
    Index 2 Param 1=0.6
    Index 2 Param 2=1.1
    Index 2 Param 3=0.1
    Index 2 Param 4=10
    Module.Param0WriteEnabled2=true
    Module.Param1WriteEnabled2=true
    Module.Param2WriteEnabled2=true
    Module.Param3WriteEnabled2=false
    Write Parameter Index 0=false
    Write Parameter Index 2=true
   ) {module=DynamicMaterialParameters001}
   &SolveForcesAndVelocity(Clamp Velocity=false, Limit Acceleration=false)
   /SetVariables(Particles.MeshOrientation=&MultiplyQuaternion(
     Quaternion A=$Particles.MeshOrientation
     Quaternion B=&MakeQuatFromAxisAngle(Axis=<0 0 0>, AngleInDegrees=&Multiply_Float(A=$Particles.Initial.SpriteRotation, B=1))
    )) {isSetParameters=true, module=SetVariables_#4}
  ]
  rendererClasses [&NiagaraMeshRendererProperties]
  renderers [{rendererClass=&NiagaraMeshRendererProperties, rendererIndex=0}]
 }
 RT_Flag001 {
  emitterSpawnScript [&FN_PreserveSkydiveUserVariables() {enabled=false, module=PreserveUserVariables}]
  emitterUpdateScript [
   &EmitterState(
    Inactive Response=::NewEnumerator1
    Life Cycle Mode=::NewEnumerator1
    Loop Behavior=::NewEnumerator1
    Loop Duration Mode=::NewEnumerator1
    UseLoopDelay=false
   )
   &SpawnBurst_Instantaneous(Spawn Count=1, Use Spawn Probability=false)
  ]
  particleSpawnScript [
   &FN_SetSpawnGroup(Override SpawnGroup=true, SpawnGroup=1)
   &InitializeParticle(
    Lifetime=100
    Mesh Scale=&MakeVector(X=0.8, Y=1, Z=2)
    Module.Write Color=true
    Module.Write Lifetime=false
    Module.Write Mass=true
    Module.Write Material Random=true
    Module.Write Position=false
    Module.Write Scale=true
    Module.Write SpriteRotation=true
    Module.Write SpriteSize=true
    Position=&SimulationPosition()
    Sprite Rotation=&UniformRangedFloat(Minimum=0, Maximum=360)
   )
   &InitialMeshOrientation(
    Orientation Coordinate Space:ENiagaraCoordinateSpace=::Local
    Orientation Vector=<0 90 0>
    Use Orientation Vector=true
    Use Rotation Vector=false
   )
   &DynamicMaterialParameters(
    Index 1 Param 3=0.75
    Module.Param0WriteEnabled1=true
    Module.Param1WriteEnabled1=true
    Module.Param2WriteEnabled1=true
    Module.Param3WriteEnabled1=true
    Write Parameter Index 0=false
    Write Parameter Index 1=true
   )
  ]
  particleUpdateScript [
   &FN_DivingAttributes_v2(SkeletalMesh:NiagaraDataInterfaceSkeletalMesh=$User.Flag_SkeletalMesh)
   &ParticleState(Kill Particles When Lifetime Has Elapsed=false)
   &UpdateMeshOrientation(Rotation Coordinate Space:ENiagaraMeshTransforms=::NewEnumerator2, Rotation Rate=0, Rotation Vector:Vector3f=<0 90 0>)
   &SampleSkeletalMeshSkeleton(
    Mesh:NiagaraDataInterfaceSkeletalMesh=$User.Flag_SkeletalMesh
    Sampled Position Coordinate Space:ENiagaraCoordinateSpace=::Local
    Sampled Rotation Coordinate Space:ENiagaraCoordinateSpace=::Local
    Sampled Velocity Coordinate Space:ENiagaraCoordinateSpace=::Local
    Spawn Mode:ENiagaraSkeletalSpawning=::NewEnumerator1
    Specific Bone Index:NiagaraInt32=$Particles.SpawnGroup
   )
   &SkeletalMeshSkeletonLocation(Module.Inherit Velocity=false, Module.Use Sampled Bone as Ribbon ID=true)
   &FN_OffsetPositionByOrientationAxis(Offset Attribute:Vector3f=$Particles.SampleSkeletalMeshSkeleton.SampledPosition, Offset Axis:Vector3f=<7 5 10>, Z Symmetry=true) {module=OffsetPositionByMeshOrientation}
   &FN_OrientMeshToEngineOwner(
    Direction:Vector3f=&FN_FrontEndVector3Modifier(Gameplay=<0 0 -1>, Front End=<0 0 -1>)
    Mesh Forward Vector:Vector3f=<0.2 1 0>
    Set Mesh Orientation=true
    Vector:Vector3f=&TransformVector(Vector=$Engine.Owner.SystemYAxis, Source Space=::World, Destination Space=::Simulation, Apply Scale=true)
   )
   &FN_MeshMaterial(
    Amplitude=80
    Falloff=HlslExpression() {hlslExpression="(Engine.Owner.Velocity.z < -2600.0f) ? 600.0f : 350.0f"}
    Frequency=300
    Jitter Multiplier=1.8545
    Offset:NiagaraFloat=$Particles.AccumulatedNormalizedDive
    Scale Noise X Axis=1.3
    Scale Noise Y Axis=2
    Scale Noise Z Axis=0
    Speed=HlslExpression() {hlslExpression="(Engine.Owner.Velocity.z < -2600.0f) ? 1.0f : 0.5f"}
   )
   &DynamicMaterialParameters(
    Index 2 Param 1=0.6
    Index 2 Param 2=1.1
    Index 2 Param 3=0.1
    Index 2 Param 4=10
    Module.Param0WriteEnabled2=true
    Module.Param1WriteEnabled2=true
    Module.Param2WriteEnabled2=true
    Module.Param3WriteEnabled2=false
    Write Parameter Index 0=false
    Write Parameter Index 2=true
   ) {module=DynamicMaterialParameters001}
   &SolveForcesAndVelocity(Clamp Velocity=false, Limit Acceleration=false)
   /SetVariables(Particles.MeshOrientation=&MultiplyQuaternion(
     Quaternion A=$Particles.MeshOrientation
     Quaternion B=&MakeQuatFromAxisAngle(Axis=<0 0 0>, AngleInDegrees=&Multiply_Float(A=$Particles.Initial.SpriteRotation, B=1))
    )) {isSetParameters=true, module=SetVariables_#4}
  ]
  rendererClasses [&NiagaraMeshRendererProperties]
  renderers [{rendererClass=&NiagaraMeshRendererProperties, rendererIndex=0}]
 }
 RT_Windburst001 {
  emitterSpawnScript [&FN_PreserveSkydiveUserVariables() {enabled=false, module=PreserveUserVariables}]
  emitterUpdateScript [
   &EmitterState(
    Inactive Response=::NewEnumerator1
    Life Cycle Mode=::NewEnumerator1
    Loop Behavior=::NewEnumerator1
    Loop Duration Mode=::NewEnumerator1
    UseLoopDelay=false
   )
   &SpawnBurst_Instantaneous(Spawn Count=1, Use Spawn Probability=false)
  ]
  particleSpawnScript [
   &FN_SetSpawnGroup(Override SpawnGroup=true, SpawnGroup=1)
   &InitializeParticle(
    Color=<rgba 1 1 1 10>
    Lifetime=100
    Mesh Scale=&MakeVector(X=0.8, Y=1, Z=4)
    Module.Write Color=true
    Module.Write Lifetime=false
    Module.Write Mass=true
    Module.Write Material Random=true
    Module.Write Position=false
    Module.Write Scale=true
    Module.Write SpriteRotation=true
    Module.Write SpriteSize=true
    Position=&SimulationPosition()
    Sprite Rotation=&UniformRangedFloat(Minimum=0, Maximum=360)
   )
   &InitialMeshOrientation(
    Orientation Coordinate Space:ENiagaraCoordinateSpace=::Local
    Orientation Vector=<0 90 0>
    Use Orientation Vector=true
    Use Rotation Vector=false
   )
  ]
  particleUpdateScript [
   &FN_DivingAttributes_v2(SkeletalMesh:NiagaraDataInterfaceSkeletalMesh=$User.Flag_SkeletalMesh)
   &ParticleState(Kill Particles When Lifetime Has Elapsed=false)
   &UpdateMeshOrientation(Rotation Coordinate Space:ENiagaraMeshTransforms=::NewEnumerator2, Rotation Rate=0, Rotation Vector:Vector3f=<0 90 0>)
   &SampleSkeletalMeshSkeleton(
    Mesh:NiagaraDataInterfaceSkeletalMesh=$User.Flag_SkeletalMesh
    Sampled Position Coordinate Space:ENiagaraCoordinateSpace=::Local
    Sampled Rotation Coordinate Space:ENiagaraCoordinateSpace=::Local
    Sampled Velocity Coordinate Space:ENiagaraCoordinateSpace=::Local
    Specific Bone Index:NiagaraInt32=$Particles.SpawnGroup
   )
   &SkeletalMeshSkeletonLocation(Module.Inherit Velocity=false, Module.Use Sampled Bone as Ribbon ID=true)
   &FN_OrientMeshToEngineOwner(
    Direction:Vector3f=&FN_FrontEndVector3Modifier(Gameplay=<0 0 -1>, Front End=<0 0 -1>)
    Mesh Forward Vector:Vector3f=<-0.2 1 0>
    Set Mesh Orientation=true
    Vector:Vector3f=&TransformVector(Vector=$Engine.Owner.SystemYAxis, Source Space=::World, Destination Space=::Simulation, Apply Scale=true)
   )
   &FN_MeshMaterial(
    Amplitude=40
    Falloff=HlslExpression() {hlslExpression="(Engine.Owner.Velocity.z < -2600.0f) ? 200.0f : 100.0f"}
    Frequency=300
    Jitter Multiplier=1.8545
    Offset:NiagaraFloat=$Particles.AccumulatedNormalizedDive
    Scale Noise X Axis=1.3
    Scale Noise Y Axis=2
    Scale Noise Z Axis=0
    Speed=HlslExpression() {hlslExpression="(Engine.Owner.Velocity.z < -2600.0f) ? 1.0f : 0.4f"}
   )
   &SolveForcesAndVelocity(Clamp Velocity=false, Limit Acceleration=false) {enabled=false}
   /SetVariables(Particles.MeshOrientation=&MultiplyQuaternion(
     Quaternion A=$Particles.MeshOrientation
     Quaternion B=&MakeQuatFromAxisAngle(Axis=<0 0 0>, AngleInDegrees=&Multiply_Float(A=$Particles.Initial.SpriteRotation, B=1))
    )) {isSetParameters=true, module=SetVariables_#4}
  ]
  rendererClasses [&NiagaraMeshRendererProperties]
  renderers [{rendererClass=&NiagaraMeshRendererProperties, rendererIndex=0}]
 }
}
summary {
 systemName=NS_Contrail_BlackKnight
 emitters [
  {emitterName=LT_ClothPieces, rendererClasses [&NiagaraSpriteRendererProperties]}
  {emitterName=RT_ClothPieces, rendererClasses [&NiagaraSpriteRendererProperties]}
  {emitterName=LT_Windburst, rendererClasses [&NiagaraMeshRendererProperties]}
  {emitterName=RT_Windburst001, rendererClasses [&NiagaraMeshRendererProperties]}
  {bEnabled=false, emitterName=LT_Flag, rendererClasses [&NiagaraMeshRendererProperties]}
  {bEnabled=false, emitterName=RT_Flag, rendererClasses [&NiagaraMeshRendererProperties]}
  {emitterName=Basic_Contrail, rendererClasses [&NiagaraRibbonRendererProperties]}
  {emitterName=LT_Flag001, rendererClasses [&NiagaraMeshRendererProperties]}
  {emitterName=RT_Flag001, rendererClasses [&NiagaraMeshRendererProperties]}
 ]
 userVariables [
  User.bIsFrontEnd {type=NiagaraBool}
  User.bIsFrontEndPreview {type=NiagaraBool}
  User.Flag_SkeletalMesh {
   type=NiagaraDataInterfaceSkeletalMesh
   defaultValue {
    struct=NiagaraExt_VariableValue_DataInterface
    value {dataInterface=&self:NiagaraDataInterfaceSkeletalMesh_2}
   }
  }
  User.PlayerSpeed {type=NiagaraFloat, defaultValue=0, description=""}
  User.RibbonIDOffset {type=NiagaraInt32}
  User.SkeletalMesh {
   type=NiagaraDataInterfaceSkeletalMesh
   defaultValue {
    struct=NiagaraExt_VariableValue_DataInterface
    value {dataInterface=&self:NiagaraDataInterfaceSkeletalMesh_1}
   }
  }
 ]
}
